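# Niagara fire effect for the LemurClam emote, attached to the emote's skeletal mesh.
# RocketFeu trails ribbons and Cheetos sheds falling sprites whose rate, offset and inherited motion respond to owner speed; both compute enlarged outline sizes.
# DirectionalBurst keeps a mesh pinned to the skeletal mesh.
# Sparkles_minimal adds flickering, three-layer flare sprites in partitioned cones, scaled by time of day, quality level and hidden hand/foot flags.
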NiagaraSystem /BRCosmetics/Animation/Game/MainPlayer/Emotes/LemurClam/VFX/NS_Emote_LemurClam_Fire
dependencies {
 usedDataInterfaces [
  &NiagaraDataInterfaceColorCurve
  &NiagaraDataInterfaceCurve
  &NiagaraDataInterfaceArrayFloat
  &NiagaraDataInterfaceSkeletalMesh
 ]
 usedDynamicInputs [
  &Sine
  &Lerp_Float
  &MakeVector2D
  &FNTimeOfDayColor
  &RandomRangeLinearColor
 ]
 usedModules [
  &self:SystemSpawnScript.SystemScriptSource.SystemScriptGraph.NiagaraNodeAssignment_10.SetVariables_#1
  &SystemState
  &self:SystemSpawnScript.SystemScriptSource.SystemScriptGraph.NiagaraNodeAssignment_0.SetVariables_#2
  &self:SystemSpawnScript.SystemScriptSource.SystemScriptGraph.NiagaraNodeAssignment_7.SetVariables_#3
  &self:SystemSpawnScript.SystemScriptSource.SystemScriptGraph.NiagaraNodeAssignment_8.SetVariables_#4
 ]
}
emitters {
 Cheetos {
  emitterUpdateScript [
   &EmitterState()
   &SpawnRate(
    SpawnRate=&MakeCustomFloatFromBool(
     Bool=&SetBoolByFloatComparison(A=&VectorLength(Vector=$User.OwnerLocalVelocity), B=50)
     True Float=40
     False Float=20
    )
    Use Spawn Probability=false
   )
  ]
  particleSpawnScript [
   &InitializeParticle(
    Color=<rgba 1 0.1667 0 1>
    Lifetime=&MaskFloatBySpawnGroup(Masked Float=0.5, Passthrough Float=5)
    Sprite Size Mode=::NewEnumerator3
    Uniform Sprite Size=5
    UsePositionOffset=false
    Write Material Random=true
   )
   &SkeletalMeshLocation(
    MaskSpawnGroup=false
    Mesh Sampling Type=::NewEnumerator5
    Orientation Sampling=::NewEnumerator1
    Position Sampling=::NewEnumerator2
    Sampled Orientation Coordinate Space=::World
    Sampled Position Coordinate Space=::World
    Sampled Position Offset=&MakeCustomVectorFromBool(
     Bool=&SetBoolByFloatComparison(A=&VectorLength(Vector=$User.OwnerLocalVelocity), B=50)
     True Vector=<0 -7 0>
     False Vector=<0 -22.5 0>
    )
    Sampled Position Offset Coordinate Space:ENiagaraSkelMeshTransforms=::NewEnumerator2
    Sampled Velocity Coordinate Space:ENiagaraCoordinateSpace=::Local
    Skeletal Mesh=$System.Skeletal Mesh
    Socket Sampling Mode=::NewEnumerator5
   ) {module=SkeletalMeshLocation002}
   &ShapeLocation(
    MaskSpawnGroup=false
    Override Seed=false
    Shape Origin=$Particles.SkeletalMeshLocation002.SampledPosition
    Sphere Radius=7.5
    UseOwnerScale=true
   )
   &AddVelocity(
    Velocity=&RandomRangeVector(Minimum=<0.5 0 -0.25>, Maximum=<-0.5 -1 0.5>)
    Velocity Speed Scale=&RandomRangeFloat(Minimum=100, Maximum=200)
   )
   &CameraOffset(Camera Offset Amount=10) {enabled=false}
  ]
  particleUpdateScript [
   &ParticleState()
   &GravityForce(Gravity=<0 0 -200>)
   &ScaleSpriteSize(Uniform Curve Sprite Scale=DataInterface() {propertyValues {Curve {
       defaultValue=1
       keys [
        {interpMode=RCIM_Cubic, tangentMode=RCTM_User, time=0, value=0}
        {interpMode=RCIM_Cubic, tangentMode=RCTM_User, time=0.5, value=1}
        {tangentMode=RCTM_User, time=1, value=0}
       ]
      }}})
   &FN_Sprite_Attributes(
    Color=&ColorFromCurve(ColorCurve=DataInterface() {propertyValues {
       AlphaCurve {keys [<at 0 1 RCIM_Cubic>,<at 1 1 RCIM_Cubic>]}
       BlueCurve {keys [<at 0 0.0052 RCIM_Cubic>]}
       GreenCurve {keys [<at 0 0.0052 RCIM_Cubic>]}
       RedCurve {keys [<at 0 0.0052 RCIM_Cubic>]}
      }})
    Create Color Attribute=true
    Write to Module Namespace=true
   ) {module=FN_Sprite_Attributes001}
   &CurlNoiseForce(
    Bias Noise=false
    Noise Frequency=150
    Noise Strength=&FloatFromCurve(
     FloatCurve=DataInterface() {propertyValues {Curve {
        defaultValue=1
        keys [<at 0 0 RCIM_Cubic>, {arriveTangent=1.5, interpMode=RCIM_Cubic, leaveTangent=1.5, time=0.1717, value=1.1353}, <at 1 1.5 RCIM_Cubic>]
       }}}
     Scale Curve=&MaskFloatBySpawnGroup(Masked Float=200, Passthrough Float=0)
    )
    Pan Noise=false
   ) {enabled=false}
   &InheritSourceMovement(
    Applied Position Scale=&MakeCustomVectorFromBool(
     Bool=&SetBoolByFloatComparison(A=&VectorLength(Vector=$User.OwnerLocalVelocity), B=300)
     True Vector=<1 1 1>
     False Vector=<0.75 0.75 0.75>
    )
    ApplyForce=false
    ApplyPosition=true
    ApplyPositionFalloff=false
   )
   &SolveForcesAndVelocity(Clamp Velocity=false, Limit Acceleration=false)
   /SetVariables(Particles.OutlineSize:Vector2f=&Multiply_Vector2D_ByFloat(Float=1.1, Vector2D=$Particles.SpriteSize)) {isSetParameters=true}
  ]
  rendererClasses [&NiagaraSpriteRendererProperties]
  renderers [
   {rendererClass=&NiagaraSpriteRendererProperties}
   {rendererClass=&NiagaraSpriteRendererProperties, rendererIndex=1}
  ]
 }
 DirectionalBurst {
  emitterUpdateScript [
   &EmitterState(
    Inactive Response=::NewEnumerator1
    Life Cycle Mode=::NewEnumerator1
    Loop Behavior=::NewEnumerator1
    Loop Duration Mode=::NewEnumerator1
    UseLoopDelay=false
   )
   &SpawnBurst_Instantaneous(Use Spawn Probability=false, UseLoopCountLimit=false)
  ]
  particleSpawnScript [
   &InitializeParticle(
    Color=<rgba 1 0.1667 0 1>
    Mass Max=2
    Mass Mode=::NewEnumerator3
    Mesh Scale Mode=::NewEnumerator3
    Mesh Uniform Scale=0.1
    Sprite Size Max=<3 14>
    Sprite Size Min=<2 6>
    Sprite Size Mode=::NewEnumerator2
    UsePositionOffset=false
    Write Material Random=true
   )
   &SkeletalMeshLocation(
    MaskSpawnGroup=false
    Mesh Sampling Type=::NewEnumerator5
    Orientation Sampling=::NewEnumerator1
    Sampled Position Offset Coordinate Space:ENiagaraSkelMeshTransforms=::NewEnumerator2
    Skeletal Mesh=$System.Skeletal Mesh
    Socket Sampling Mode=::NewEnumerator5
   )
  ]
  particleUpdateScript [
   &ParticleState(Kill Particles When Lifetime Has Elapsed=false)
   &SkeletalMeshLocation(
    MaskSpawnGroup=false
    Mesh Sampling Type=::NewEnumerator5
    Sampled Position Offset=<0 -3.4956 0>
    Sampled Position Offset Coordinate Space:ENiagaraSkelMeshTransforms=::NewEnumerator2
    Skeletal Mesh=$System.Skeletal Mesh
    Socket Sampling Mode=::NewEnumerator5
   ) {module=SkeletalMeshLocation002}
  ]
  rendererClasses [&NiagaraMeshRendererProperties]
  renderers [{rendererClass=&NiagaraMeshRendererProperties, rendererIndex=0}]
 }
 RocketFeu {
  emitterUpdateScript [
   &EmitterState()
   &SpawnRate(SpawnRate=35, Use Spawn Probability=false)
  ]
  particleSpawnScript [
   &InitializeParticle(
    Color=<rgba 1 0.1667 0 1>
    Lifetime=&MaskFloatBySpawnGroup(Masked Float=0.3, Passthrough Float=5)
    Ribbon Width=100
    Ribbon Width Mode=::NewEnumerator1
    UsePositionOffset=false
    Write Material Random=true
   )
   &SkeletalMeshLocation(
    MaskSpawnGroup=false
    Mesh Sampling Type=::NewEnumerator5
    Orientation Sampling=::NewEnumerator1
    Position Sampling=::NewEnumerator2
    Sampled Orientation Coordinate Space=::World
    Sampled Position Coordinate Space=::World
    Sampled Position Offset Coordinate Space:ENiagaraSkelMeshTransforms=::NewEnumerator1
    Sampled Velocity Coordinate Space:ENiagaraCoordinateSpace=::Local
    Skeletal Mesh=$System.Skeletal Mesh
    Socket Sampling Mode=::NewEnumerator5
   ) {module=SkeletalMeshLocation002}
   /SetVariables(Particles.Position=&AddVectorToPosition(Position=$Particles.SkeletalMeshLocation002.SampledPosition, Vector=<0 0 0>)) {isSetParameters=true}
   &AddVelocity(Velocity=<0 -200 0>, Velocity Speed Scale=&MaskFloatBySpawnGroup(Masked Float=1, Passthrough Float=0))
   /SetVariables(Particles.RibbonID=/Niagara/DynamicInputs/Meshes/MakeRibbonIDFromSpawnGroup()) {isSetParameters=true}
   &CameraOffset(Camera Offset Amount=10) {enabled=false}
  ]
  particleUpdateScript [
   &ParticleState()
   &FN_Sprite_Attributes(
    Camera Offset=-10
    Color=&ColorFromCurve(ColorCurve=DataInterface() {propertyValues {
       AlphaCurve {keys [<at 0 1 RCIM_Cubic>,<at 1 1 RCIM_Cubic>]}
       BlueCurve {keys [<at 0 0.0052 RCIM_Cubic>]}
       GreenCurve {keys [<at 0 0.0052 RCIM_Cubic>]}
       RedCurve {keys [<at 0 0.0052 RCIM_Cubic>]}
      }})
    Create Camera Offset Attribute=true
    Create Color Attribute=true
    Write to Module Namespace=true
   ) {module=FN_Sprite_Attributes001}
   &ScaleRibbonWidth(Ribbon Width Scale=&FloatFromCurve(
     FloatCurve=DataInterface() {propertyValues {Curve {keys [
         {arriveTangent=0.2671, interpMode=RCIM_Cubic, leaveTangent=0.2671, tangentMode=RCTM_User, time=-0.002, value=0.1229}
         {arriveTangent=0.0019, interpMode=RCIM_Cubic, leaveTangent=0.0019, tangentMode=RCTM_User, time=0.5296, value=0.1976}
         {arriveTangent=-0.0402, interpMode=RCIM_Cubic, leaveTangent=-0.0402, tangentMode=RCTM_User, time=0.95, value=0}
        ]}}}
     Scale Curve=&RandomRangeFloat(Minimum=0.9, Maximum=1.5)
    )) {module=ScaleRibbonWidth001}
   &CurlNoiseForce(
    Bias Noise=false
    Noise Frequency=150
    Noise Strength=&FloatFromCurve(
     FloatCurve=DataInterface() {propertyValues {Curve {
        defaultValue=1
        keys [<at 0 0 RCIM_Cubic>, {arriveTangent=1.5, interpMode=RCIM_Cubic, leaveTangent=1.5, time=0.1717, value=1.1353}, <at 1 1.5 RCIM_Cubic>]
       }}}
     Scale Curve=&MaskFloatBySpawnGroup(Masked Float=200, Passthrough Float=0)
    )
    Pan Noise=false
   ) {enabled=false}
   &InheritSourceMovement(ApplyForce=false, ApplyPosition=true, ApplyPositionFalloff=false)
   &SolveForcesAndVelocity(Clamp Velocity=false, Limit Acceleration=false)
   /SetVariables(Particles.OutlineWidth:NiagaraFloat=&Multiply_Float(A=1.0892, B=$Particles.RibbonWidth)) {isSetParameters=true}
  ]
  rendererClasses [&NiagaraRibbonRendererProperties]
  renderers [
   {rendererClass=&NiagaraRibbonRendererProperties}
   {rendererClass=&NiagaraRibbonRendererProperties, rendererIndex=1}
  ]
 }
 Sparkles_minimal {
  emitterUpdateScript [
   &EmitterState()
   &SpawnRate(SpawnRate=60, Use Spawn Probability=false)
  ]
  particleSpawnScript [
   &PartitionParticles(Number Of Partitions=3, Partition Mode:ENiagara_PartitionMode_v2=::NewEnumerator5)
   /SetVariables(Particles.PartitionParticles.Partition:NiagaraInt32=&RandomRangeInt(Minimum=0, Maximum=2)) {isSetParameters=true}
   &InitializeParticle(
    Lifetime Max=0.7
    Lifetime Min=0.5
    Lifetime Mode=::NewEnumerator1
    Sprite Size Mode=::NewEnumerator3
    Uniform Sprite Size=&RandomRangeFloat(Minimum=15, Maximum=25)
    UsePositionOffset=false
    Write Material Random=true
   )
   &ShapeLocation(
    Cone Angle:NiagaraFloat=&MergeHelper_FloatSwitch(0=32.0498, 1=18.2547, 2=9.9033, 3=?, Index=$Particles.PartitionParticles.Partition, Num=::NewEnumerator1)
    Cone Inner Angle:NiagaraFloat=&MergeHelper_FloatSwitch(0=55.7635, 1=42.3344, 2=43.0964, 3=?, Index=$Particles.PartitionParticles.Partition, Num=::NewEnumerator1)
    Cone Length:NiagaraFloat=&MergeHelper_FloatSwitch(0=42.9057, 1=61.4188, 2=58.522, 3=?, Index=$Particles.PartitionParticles.Partition, Num=::NewEnumerator1)
    Flatten Endcaps=true
    MaskSpawnGroup=false
    Non Uniform Scale=&MergeHelper_VectorSwitch(0=<1 1 1>, 1=<1 1 1>, 2=<1 1 1>, 3=?, Index=$Particles.PartitionParticles.Partition, Num=::NewEnumerator1)
    Offset=&MergeHelper_VectorSwitch(0=<0 60 0>, 1=<0 -58.3629 0>, 2=<0 61 0>, 3=?, Index=$Particles.PartitionParticles.Partition, Num=::NewEnumerator1)
    Offset Mode=::NewEnumerator0
    Override Seed=false
    Rotation Mode=::NewEnumerator2
    Shape Primitive=::NewEnumerator5
    UseOwnerScale=true
    Yaw / Pitch / Roll=&MergeHelper_VectorSwitch(0=<0 0 -0.25>, 1=<0 0 0.25>, 2=<0 0 -0.25>, 3=?, Index=$Particles.PartitionParticles.Partition, Num=::NewEnumerator1)
   ) {module=ShapeLocation001}
  ]
  particleUpdateScript [
   &ParticleState()
   &ScaleSpriteSize(Uniform Curve Scale=&Multiply_Float(
     A=&FNTimeOfDayScalar(Night Scale=1.25, Morning Scale=1, Evening Scale=1)
     B=&Multiply_Float(
      A=&MakeCustomFloatFromBool(
       Bool=$User.bisHandHidden
       True Float=&SelectFloatFromArray(
        Direct Array Index=$Particles.PartitionParticles.Partition
        Float Selection Array=DataInterface() {propertyValues {FloatData [1,1,1,0,0,1,1,1,1]}}
       )
       False Float=1
      )
      B=&MakeCustomFloatFromBool(
       Bool=$User.bisFootHidden
       True Float=&SelectFloatFromArray(
        Direct Array Index=$Particles.PartitionParticles.Partition
        Float Selection Array=DataInterface() {propertyValues {FloatData [1,1,1,1,1,1,1,0,0]}}
       )
       False Float=1
      )
     )
    )) {module=ScaleSpriteSize001}
   &SkeletalMeshLocation(
    Bone Sampling Mode=::NewEnumerator4
    MaskSpawnGroup=false
    Orientation Sampling=::NewEnumerator1
    Sampled Position Offset:NiagaraPosition=$Particles.ShapeLocation001.ShapePosition
    Skeletal Mesh=$System.Skeletal Mesh
    Use Sampled Mesh Attribute as Ribbon ID=false
   ) {module=SkeletalMeshLocation002}
   &ShapeLocation(
    Cylinder Height=&MergeHelper_FloatSwitch(0=-7, 1=23, 2=-23, 3=27, 4=-27, 5=-25, 6=31, 7=23, 8=12, 9=8, Index=$Particles.PartitionParticles.Partition, Num=::NewEnumerator8)
    Cylinder Height Midpoint=1
    Cylinder Radius=&MergeHelper_FloatSwitch(0=4.5, 1=3, 2=3, 3=2, 4=2, 5=7.5, 6=7.5, 7=10, 8=13, 9=14, Index=$Particles.PartitionParticles.Partition, Num=::NewEnumerator8)
    MaskSpawnGroup=false
    Override Seed=false
    Shape Primitive=::NewEnumerator1
    Surface Only=true
    Surface Only Band Thickness=0.3
    Use Endcaps In Surface Only Mode=false
    UseOwnerScale=true
   ) {enabled=false}
   &ScaleColor(
    Scale Alpha=&Multiply_Float(
     A=&FloatFromCurve(
      FloatCurve=DataInterface() {propertyValues {Curve {keys [
          {interpMode=RCIM_Cubic, tangentMode=RCTM_User, time=0, value=0}
          {arriveTangent=-2.4058, interpMode=RCIM_Cubic, leaveTangent=-2.4058, tangentMode=RCTM_User, time=0.26, value=0.8421}
          {arriveTangent=-1.1775, interpMode=RCIM_Cubic, leaveTangent=-1.1775, time=0.4613, value=0.0645}
          {arriveTangent=-0.1197, interpMode=RCIM_Cubic, leaveTangent=-0.1197, time=0.5898, value=0.4537}
          <at 1 0 RCIM_Cubic>
         ]}}}
      Scale Curve=&RandomRangeFloat(Minimum=0.1, Maximum=0.5)
     )
     B=&FloatFromCurve(
      FloatCurve=DataInterface() {propertyValues {Curve {keys [{tangentMode=RCTM_User, time=0, value=0}, {tangentMode=RCTM_User, time=1, value=1}]}}}
      Scale Curve=&FloatFromCurve(
       FloatCurve=DataInterface() {propertyValues {Curve {keys [
           {tangentMode=RCTM_User, time=0, value=0}
           {arriveTangentWeight=0.1667, interpMode=RCIM_Cubic, leaveTangentWeight=0.1667, tangentMode=RCTM_User, time=1, value=1}
          ]}}}
       CurveIndex=$Emitter.Age
      )
     )
    )
    Scale RGB=&VectorFromFloat(Value=&Multiply_Float(
      A=&RandomRangeFloat(Minimum=0.7, Maximum=0.9, Evaluation Type=::NewEnumerator1)
      B=&SelectFloatFromArray(Direct Array Index=$Engine.QualityLevel, Float Selection Array=DataInterface() {propertyValues {FloatData [10,5,1,1,1]}})
     ))
    ScaleA=true
    ScaleRGB=true
   ) {module=ScaleColor001}
   &ScaleColor(
    Scale Alpha=$System.TODAlpha
    Scale RGB=&MakeVectorFromLinearColorRGB(Color=$System.TODColor)
    ScaleA=true
    ScaleRGB=true
   ) {enabled=false}
   /SetVariables(
    Particles.CenterFlareColor:LinearColor=&MultiplyLinearColor(
     Linear Color=$Particles.Color
     Scale Factor=&MakeVector4_FromLinearColor(LinearColor=&LerpLinearColors(StartColor=<rgba 8.0099 9.8007 10 1>, EndColor=<rgba 3.4218 8.799 10 1>, LerpFactor=0))
    )
    Particles.CenterFlareSize:Vector2f=&Multiply_Vector2(A=$Particles.SpriteSize, B=&Vector2DFromFloat(Value=0.5))
    Particles.CoronaFlareColor:LinearColor=&MultiplyLinearColor(
     Linear Color=$Particles.Color
     Scale Factor=&MakeVector4_FromLinearColor(LinearColor=&LerpLinearColors(StartColor=<rgba 0.15 0.1242 0.1057 1>, EndColor=<rgba 0.0729 0.0489 0.0317 1>, LerpFactor=0))
    )
    Particles.CoronaFlareSize:Vector2f=&Multiply_Vector2(A=$Particles.SpriteSize, B=&Vector2DFromFloat(Value=0.75))
    Particles.FlickerFlareColor:LinearColor=&MultiplyLinearColor(
     Linear Color=$Particles.Color
     Scale Factor=&MakeVector4_FromLinearColor(LinearColor=&LerpLinearColors(StartColor=<rgba 0.9441 0.8079 1 1>, EndColor=<rgba 0.0393 0.0337 0.0417 50>, LerpFactor=0))
    )
   ) {isSetParameters=true}
   &CameraOffset(Camera Offset Amount=20) {module=CameraOffset001}
  ]
  rendererClasses [&NiagaraSpriteRendererProperties]
  renderers [
   {rendererClass=&NiagaraSpriteRendererProperties}
   {rendererClass=&NiagaraSpriteRendererProperties, rendererIndex=1}
   {rendererClass=&NiagaraSpriteRendererProperties, rendererIndex=2}
  ]
 }
}
summary {
 systemName=NS_Emote_LemurClam_Fire
 emitters [
  {emitterName=RocketFeu, rendererClasses [&NiagaraRibbonRendererProperties]}
  {emitterName=Cheetos, rendererClasses [&NiagaraSpriteRendererProperties]}
  {emitterName=DirectionalBurst, rendererClasses [&NiagaraMeshRendererProperties]}
  {emitterName=Sparkles_minimal, rendererClasses [&NiagaraSpriteRendererProperties]}
 ]
 userVariables [
  User.bisFootHidden {type=NiagaraBool}
  User.bIsFrontEnd {type=NiagaraBool}
  User.bIsFrontEndPreview {type=NiagaraBool}
  User.bisHandHidden {type=NiagaraBool}
  User.Emote_DuckCoast_FireVelX {type=NiagaraFloat, defaultValue=0, description=""}
  User.Emote_DuckCoast_FireVelY {type=NiagaraFloat, defaultValue=0, description=""}
  User.Emote_DuckCoast_FireVelZ {type=NiagaraFloat, defaultValue=0, description=""}
  User.OwnerLocalVelocity {type=Vector3f, description="", defaultValue {struct=Vector3f, value=<0 0 0>}}
  User.RibbonIDOffset {type=NiagaraInt32}
  User.SwingSectionIndex {type=NiagaraInt32}
  User.Variant {type=NiagaraFloat, defaultValue=0, description=""}
 ]
}
